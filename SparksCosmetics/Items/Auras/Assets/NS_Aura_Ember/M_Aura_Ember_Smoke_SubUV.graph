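# Translucent unlit Niagara sprite material for smoke in the Ember cosmetic aura effect.
# Opacity multiplies particle alpha by the red channel of a SubUV sprite texture and a tiled inky smoke noise, softened by a lerp parameter, then applies depth fade.
# Emissive takes particle color, or a Color parameter via static switch, through the MF_FX_EmissiveAdjustments function.
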Material /SparksCosmetics/Items/Auras/Assets/NS_Aura_Ember/M_Aura_Ember_Smoke_SubUV
graph {
 properties {bUsedWithNiagaraSprites=true, blendMode=BLEND_Translucent, pixelDepthOffsetMode=PDOM_Legacy, shadingModel=MSM_Unlit, usedShadingModels=Unlit}
 nodes [
  n1 = ParticleColor()
  n2 = TextureSampleParameter2D() {parameterName=SubUV_Tex, texture=/BRCosmetics/Characters/Player/Female/Medium/Bodies/F_MED_MadameMoth_Posh/FX/T_MadameMoth_fire_sprites}
  n3 = LinearInterpolate(
   A=TextureSample(UVs=TextureCoordinate() {uTiling=1.5, vTiling=1.5}) {texture=/Game/Effects/Fort_Effects/Textures/Tile/T_Inky_Smoke_Tile2}.R
   Alpha=ScalarParameter() {defaultValue=0.1, parameterName=Noise Lerp}
  ) {constAlpha=0.1}
  DepthFade(Opacity=Multiply(A=@n1.A, B=Multiply(A=@n2.R, B=@n3)), FadeDistance=ScalarParameter() {defaultValue=20, parameterName=FadeDistance}) {fadeDistanceDefault=5}
  MaterialFunctionCall(Emissive=StaticSwitchParameter(True=@n1.RGB, False=VectorParameter() {parameterName=Color, defaultValue=<rgba 0 0 0 1>}.RGB) {defaultValue=true, parameterName=UseParticleColor}) {materialFunction=&MF_FX_EmissiveAdjustments}
 ]
}
