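# Lit material for glowing stone shard mesh particles in the Ember aura Niagara cosmetic effect.
# A tombstone texture passes through ConvertFromDiffSpec to give a base color.
# The UseParticleColor scalar blends that color toward a version tinted by particle color.
# Particle color also feeds MF_FX_EmissiveAdjustments for the emissive glow, so the emitter drives the glow color.
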Material /SparksCosmetics/Items/Auras/Assets/NS_Aura_Ember/M_Stone_Shards_Emissive
graph {
 properties {bUsedWithNiagaraMeshParticles=true, pixelDepthOffsetMode=PDOM_Legacy, usedShadingModels=Default Lit}
 nodes [
  n1 = MaterialFunctionCall(DiffuseColor=TextureSample() {texture=/Game/Environments/Props/Spiritual/Textures/t_Tombstone_Solid}.RGB) {materialFunction=&ConvertFromDiffSpec}
  n2 = ParticleColor()
  LinearInterpolate(A=@n1.BaseColor, B=Multiply(A=@n2.RGB, B=@n1.BaseColor), Alpha=ScalarParameter() {parameterName=UseParticleColor})
  MaterialFunctionCall(Emissive=@n2.RGB) {materialFunction=&MF_FX_EmissiveAdjustments}
 ]
}
